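# Masked, Nanite-enabled telephone booth material that adds procedural blinds, opened and closed by a BlindsAperture custom primitive data value.
# Normalized local height is cut at the aperture for masking and shading; diffuse alpha separates booth textures from blinds.
# Blinds normal noise uses side-selected local projection, sliding and changing strength with aperture.
# World position offset combines destructible damage, adjusted by BlindsAreAnimating, tree animation and optional vertex-color offset, disabled on mobile.
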Material /Game/Gadgets/Assets/Telephone_Booth/Materials/M_TelephoneBooth
graph {
 properties {bUsedWithNanite=true, blendMode=BLEND_Masked, pixelDepthOffsetMode=PDOM_Legacy, usedShadingModels=Default Lit}
 nodes [
  n1 = Divide(
   A=MaterialFunctionCall() {materialFunction=&LocalPosition}.Local Position
   B=VectorParameter() {parameterName=Bounds, defaultValue=<rgba 121 121 258 0>}.RGB
  )
  n2 = PreSkinnedNormal()
  n3 = Constant3Vector() {constant=<rgba 1 0 0 0>}
  n4 = MaterialFunctionCall!(In=Abs(None=DotProduct(A=@n2, B=@n3))) {materialFunction=/Engine/Functions/Engine_MaterialFunctions02/Utility/Round}
  n5 = LinearInterpolate(
   A=ComponentMask(None=@n1) {b=true, r=true}
   B=ComponentMask(None=@n1) {b=true, g=true}
   Alpha=ComponentMask(None=VertexInterpolator(VS=@n4.Result).PS) {r=true}
  )
  n6 = Min(A=Max(A=ScalarParameter() {bUseCustomPrimitiveData=true, defaultValue=0.3321, parameterName=BlindsAperture, primitiveDataIndex=5}) {constB=0.0857}) {constB=0.75}
  n7 = AppendVector(A=Constant(), B=Multiply(A=OneMinus(None=@n6), B=ScalarParameter() {defaultValue=10, parameterName=test}))
  n8 = VectorParameter() {parameterName=Normal Tiling, defaultValue=<rgba 0.15 3 0 0>}
  n9 = TextureSample(UVs=Multiply(A=Add(A=@n5, B=@n7), B=AppendVector(A=@n8.R, B=@n8.R))) {constMipValue=4, samplerType=SAMPLERTYPE_Normal, texture=/Game/Gadgets/Assets/Telephone_Booth/Textures/T_BlindsNormalNoise_N}
  n10 = ScalarParameter() {defaultValue=2, parameterName=BlindsNormalIntensity}
  n11 = Multiply(
   A=ComponentMask(None=@n9.RGB) {g=true, r=true}
   B=Power(Base=Multiply(A=@n6, B=@n10), Exp=ScalarParameter() {defaultValue=1, parameterName=BlindsNormalAttenuation})
  )
  n12 = ComponentMask(None=@n9.RGB) {b=true}
  n13 = TextureSample() {texture=/Game/Gadgets/Assets/Telephone_Booth/Textures/T_Item_Telephone_Booth_D}
  LinearInterpolate(
   A=Normalize(VectorInput=AppendVector(A=@n11, B=@n12))
   B=TextureSample() {samplerType=SAMPLERTYPE_Normal, texture=/Game/Gadgets/Assets/Telephone_Booth/Textures/T_Item_Telephone_Booth_N}.RGB
   Alpha=@n13.A
  )
  LinearInterpolate(
   A=ScalarParameter() {defaultValue=0.8, parameterName=BlindsRoughness}
   B=TextureSample() {samplerType=SAMPLERTYPE_Masks, texture=/Game/Gadgets/Assets/Telephone_Booth/Textures/T_Item_Telephone_Booth_S}.B
   Alpha=@n13.A
  )
  n14 = Subtract(A=Time(), B=ScalarParameter() {bUseCustomPrimitiveData=true, group=ActivationSwitch, parameterName=GameTimeFromBP, primitiveDataIndex=6})
  n15 = ScalarParameter() {bUseCustomPrimitiveData=true, defaultValue=0.5, group=ActivationSwitch, parameterName=AnimationDuration, primitiveDataIndex=3}
  n16 = Add(A=@n15) {constB=0}
  n17 = Power(
   Base=Saturate(None=Divide(A=@n14, B=@n16))
   Exp=ScalarParameter() {defaultValue=2, group=ActivationSwitch, parameterName=Animation_EaseInOut}
  ) {constExponent=8}
  LinearInterpolate(A=OneMinus(None=@n17), B=@n17, Alpha=@n6)
  MaterialFunctionCall() {materialFunction=&ObjectLocalBounds}
  n18 = ScalarParameter() {
   bUseCustomPrimitiveData=true
   defaultValue=0.3321
   group=z_GameplayParameters_DestructibleDamage
   parameterName=BlindsAreAnimating
   primitiveDataIndex=6
  }
  Reroute(None=@n18)
  n19 = ComponentMask(None=@n1) {b=true}
  n20 = Add(A=Power(Base=@n19, Exp=ScalarParameter() {defaultValue=0.66, parameterName=BlindsGradientContrast}), B=Multiply(A=@n6) {constB=-1})
  n21 = Add(A=Multiply(A=Saturate(None=@n20), B=ScalarParameter() {defaultValue=0.5, parameterName=BlindsGradientBrightness}), B=@n13.RGB)
  n22 = VectorParameter() {parameterName=BlindsColor, defaultValue=<rgba 0.9219 0.7134 0.4465 0>}
  Reroute(None=LinearInterpolate(A=Multiply(A=@n21, B=@n22.RGB), B=@n13.RGB, Alpha=@n13.RGB))
  n23 = Add(A=@n6) {constB=-0.01}
  Saturate(None=Add(
    A=MaterialFunctionCall(Alpha=@n19, Min=@n23, Max=@n6) {materialFunction=&SmoothStep}.Result
    B=@n13.A
   ))
  ScalarParameter() {defaultValue=0.1381, parameterName=BlindsApertureTest, primitiveDataIndex=3}
  n24 = VectorParameter() {bUseCustomPrimitiveData=true, group=z_GameplayParameters_DestructibleDamage, parameterName=WSDeformationCenter, primitiveDataIndex=20}
  n25 = VectorParameter() {group=z_GameplayParameters_DestructibleDamage, parameterName=WSDeformationCenterOffset, defaultValue=<rgba 0 0 50 0>}
  n26 = Multiply(A=@n18, B=@n25.RGB)
  n27 = ScalarParameter() {
   bUseCustomPrimitiveData=true
   defaultValue=1
   group=z_GameplayParameters_DestructibleDamage
   parameterName=WSDamageRadius
   primitiveDataIndex=26
  }
  n28 = LinearInterpolate(B=ScalarParameter() {defaultValue=0.75, group=z_GameplayParameters_DestructibleDamage, parameterName=WSDamageRadiusMultiplier}, Alpha=@n18) {constA=1}
  n29 = Add(
   A=MaterialFunctionCall(WSDeformationCenter=Add(A=@n24.RGB, B=@n26), WSDamageRadius=Multiply(A=@n27, B=@n28)) {materialFunction=&MF_FN_DestructibleDamage}.DamageAnimWPO
   B=MaterialFunctionCall!(Use Tree Vertex Shader=StaticBoolParameter() {group=Special Cases, parameterName=Use Tree Vertex Shader}) {materialFunction=/Game/Packages/Fortress_SharedMaterials/Base_Material_Functions/FortniteTreeAnimation_VS}.Result
  )
  n30 = VectorParameter() {group=WorldPositionOffset, parameterName=WorldPositionOffset_ModulatedByVertexColor, defaultValue=<rgba 0 0 0 1>}
  n31 = VertexColor()
  n32 = Multiply(
   A=@n30.RGB
   B=StaticComponentMaskParameter(None=AppendVector(A=@n31, B=@n31)) {defaultA=true, group=WorldPositionOffset, parameterName=WorldPositionOffsetVertexColorModulationChannel}
  )
  ShadingPathSwitch(
   Default=StaticSwitchParameter(True=Add(A=@n29, B=@n32), False=@n29) {group=WorldPositionOffset, parameterName=UseWorldPositionOffsetModulatedByVertexColors}
   Mobile=Constant()
  ) {desc=Migrated from FeatureLevelSwitch}
 ]
}
