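# Translucent, two-sided, lit glass material for the telephone booth gadget, with refraction and parameters for color, opacity, roughness and specular.
# An offset graph sums destructible damage animation, driven by custom primitive data, with a tree vertex animation function.
# BlindsAreAnimating blends in a deformation center offset and radius multiplier.
# A static switch can add vertex-color-modulated offset, and a shading path switch zeroes the result on mobile.
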Material /Game/Gadgets/Assets/Telephone_Booth/Materials/M_TelephoneBoothGlass
graph {
 properties {
  bUsedWithNanite=true
  blendMode=BLEND_Translucent
  pixelDepthOffsetMode=PDOM_Legacy
  refractionMethod=RM_IndexOfRefraction
  translucencyLightingMode=TLM_Surface
  twoSided=true
  usedShadingModels=Default Lit
 }
 nodes [
  ScalarParameter() {defaultValue=0.33, parameterName=Opacity}
  ScalarParameter() {defaultValue=1, parameterName=Refraction}
  ScalarParameter() {parameterName=Roughness}
  ScalarParameter() {defaultValue=1, parameterName=Specular}
  n1 = VectorParameter() {bUseCustomPrimitiveData=true, group=z_GameplayParameters_DestructibleDamage, parameterName=WSDeformationCenter, primitiveDataIndex=20}
  n2 = ScalarParameter() {
   bUseCustomPrimitiveData=true
   defaultValue=0.3321
   group=z_GameplayParameters_DestructibleDamage
   parameterName=BlindsAreAnimating
   primitiveDataIndex=6
  }
  n3 = VectorParameter() {group=z_GameplayParameters_DestructibleDamage, parameterName=WSDeformationCenterOffset, defaultValue=<rgba 0 0 50 0>}
  n4 = Multiply(A=@n2, B=@n3.RGB)
  n5 = ScalarParameter() {
   bUseCustomPrimitiveData=true
   defaultValue=1
   group=z_GameplayParameters_DestructibleDamage
   parameterName=WSDamageRadius
   primitiveDataIndex=26
  }
  n6 = LinearInterpolate(B=ScalarParameter() {defaultValue=0.75, group=z_GameplayParameters_DestructibleDamage, parameterName=WSDamageRadiusMultiplier}, Alpha=@n2) {constA=1}
  n7 = Add(
   A=MaterialFunctionCall(WSDeformationCenter=Add(A=@n1.RGB, B=@n4), WSDamageRadius=Multiply(A=@n5, B=@n6)) {materialFunction=&MF_FN_DestructibleDamage}.DamageAnimWPO
   B=MaterialFunctionCall!(Use Tree Vertex Shader=StaticBoolParameter() {group=Special Cases, parameterName=Use Tree Vertex Shader}) {materialFunction=/Game/Packages/Fortress_SharedMaterials/Base_Material_Functions/FortniteTreeAnimation_VS}.Result
  )
  n8 = VectorParameter() {group=WorldPositionOffset, parameterName=WorldPositionOffset_ModulatedByVertexColor, defaultValue=<rgba 0 0 0 1>}
  n9 = VertexColor()
  n10 = Multiply(
   A=@n8.RGB
   B=StaticComponentMaskParameter(None=AppendVector(A=@n9, B=@n9)) {defaultA=true, group=WorldPositionOffset, parameterName=WorldPositionOffsetVertexColorModulationChannel}
  )
  ShadingPathSwitch(
   Default=StaticSwitchParameter(True=Add(A=@n7, B=@n10), False=@n7) {group=WorldPositionOffset, parameterName=UseWorldPositionOffsetModulatedByVertexColors}
   Mobile=Constant()
  ) {desc=Migrated from FeatureLevelSwitch}
  VectorParameter() {parameterName=Color, defaultValue=<rgba 0.0738 0.0885 0.1042 0>}
 ]
}
